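# Looping ambience cue for an Arctic tile mix in Creative, routed to the Creative ambient tile-mix sound class.
# A quality-level switch plays a winter day loop mixed with a quieter desert day bed on the first two levels.
# The last two levels instead play a single optimized winter evening loop, slightly pitched down with varied volume.
# A large box attenuation covers the area, and concurrency allows one instance, preventing new ones.
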SoundCue /Game/Sounds/Fort_Ambient_Looping/Tile_Mixes/Temperate_Creative_Arctic_AMB_Cue
root=QualityLevel_0
nodes [
 {id=Mixer_1, children [WavePlayer_6,WavePlayer_7], properties {inputVolume [1,0.3]}}
 {id=Modulator_2, children [WavePlayer_3], properties {pitchMax=0.9, pitchMin=0.85, volumeMax=1.1, volumeMin=0.9}}
 {id=QualityLevel_0, children [Mixer_1,Mixer_1,Modulator_2,Modulator_2]}
 {id=WavePlayer_3, properties {bLooping=true, soundWaveAssetPtr=/Game/Athena/Sounds/Ambient/Temperate_Winter_Evening_Loop_Opt}}
 {id=WavePlayer_6, properties {bLooping=true, soundWaveAssetPtr=/Game/Athena/Sounds/Ambient/Amb_Winter_Day_Loop_r2}}
 {id=WavePlayer_7, properties {bLooping=true, soundWaveAssetPtr=/Game/Sounds/Fort_Ambient_Looping/Desert/Desert_Ambient_Day_01}}
]
properties {
 bOverrideAttenuation=true
 bOverrideConcurrency=true
 maxDistance=33000
 soundClassObject=/Game/Sounds/FortSoundClassesAndModes/Fort_Creative_Ambient_TileMix
 volumeMultiplier=1
 attenuationOverrides {attenuationShape=Box, falloffDistance=8000, attenuationShapeExtents=<25000 25000 25000>}
 concurrencyOverrides {maxCount=1, resolutionRule=PreventNew}
}
