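# Looping ambient wind bed sound cue, named for The Block in temperate Creative, routed to the Creative ambient tile-mix sound class.
# A quality-level switch randomly picks one of two looping mountain wind waves in its first two slots, and uses a single "opt" wind loop in the other two.
# Overridden box attenuation covers a large area, and concurrency allows one instance, preventing new ones from starting.
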SoundCue /Game/Sounds/Fort_Ambient_Looping/Tile_Mixes/Temperate_Creative_TheBlock_AMB_Cue
root=QualityLevel_0
nodes [
 {id=QualityLevel_0, children [Random_0,Random_0,WavePlayer_3,WavePlayer_3]}
 {id=Random_0, children [WavePlayer_1,WavePlayer_2], properties {weights [1,1]}}
 {id=WavePlayer_1, properties {bLooping=true, soundWaveAssetPtr=/Game/Sounds/Fort_Ambient_Looping/Mountain_Exterior/Mountain_Wind01}}
 {id=WavePlayer_2, properties {bLooping=true, soundWaveAssetPtr=/Game/Sounds/Fort_Ambient_Looping/Mountain_Exterior/Mountain_Wind06}}
 {id=WavePlayer_3, properties {bLooping=true, soundWaveAssetPtr=/Game/Sounds/Fort_Ambient_Looping/Mountain_Exterior/Mountain_Wind06_opt}}
]
properties {
 bOverrideAttenuation=true
 bOverrideConcurrency=true
 maxDistance=33000
 soundClassObject=/Game/Sounds/FortSoundClassesAndModes/Fort_Creative_Ambient_TileMix
 volumeMultiplier=0.8
 attenuationOverrides {attenuationShape=Box, falloffDistance=8000, attenuationShapeExtents=<25000 25000 25000>}
 concurrencyOverrides {maxCount=1, resolutionRule=PreventNew}
}
